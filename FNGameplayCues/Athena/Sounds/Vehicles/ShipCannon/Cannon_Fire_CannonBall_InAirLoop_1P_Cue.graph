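# Looping sound cue for a ship cannon's cannonball in flight, named as the first-person variant.
# A looping wave player feeds a modulator that applies a fixed raised pitch, and the cue plays at reduced volume.
# It uses a third-person ranged weapon fire sound class despite the 1P name.
# Its own concurrency override is set to prevent new instances once the limit is reached.
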SoundCue /FNGameplayCues/Athena/Sounds/Vehicles/ShipCannon/Cannon_Fire_CannonBall_InAirLoop_1P_Cue
root=Modulator_2
nodes [
 {id=Modulator_2, children [WavePlayer_1], properties {pitchMax=1.3, pitchMin=1.3}}
 {
  id=WavePlayer_1
  properties {bLooping=true, soundWaveAssetPtr=&/Cannon_Fire_CannonBall_InAirLoop_01}
 }
]
properties {
 bOverrideConcurrency=true
 maxDistance=2097152
 soundClassObject=&Fort_WeaponRanged_Fire_3P_PSM
 volumeMultiplier=0.5
 attenuationOverrides {falloffDistance=5000, reverbWetLevelMax=0.7, reverbWetLevelMin=0.2, attenuationShapeExtents {x=1000}}
 concurrencyOverrides {resolutionRule=PreventNew}
}
