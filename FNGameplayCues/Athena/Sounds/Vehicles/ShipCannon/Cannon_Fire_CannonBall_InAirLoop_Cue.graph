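# Sound cue for the in-air loop of a fired ship cannon cannonball.
# Mixes a dominant, pitched-up cannonball loop on medium projectile attenuation with a quieter quad rocket loop on distant attenuation, each with its own sound class.
# The rocket layer adds pitch variation and a volume envelope that ramps it in.
# Cue-level concurrency override prevents new instances once the limit is reached.
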SoundCue /FNGameplayCues/Athena/Sounds/Vehicles/ShipCannon/Cannon_Fire_CannonBall_InAirLoop_Cue
root=Mixer_0
nodes [
 {
  id=Attenuation_0
  children [Modulator_2]
  properties {
   attenuationSettings=&Fort_Weapon_Projectile_Medium_Mono_Athena_Att
   attenuationOverrides {bAttenuate=false, falloffDistance=10000, stereoSpread=800, attenuationShapeExtents {x=2000}}
  }
 }
 {
  id=Attenuation_1
  children [Modulator_0]
  properties {
   attenuationSettings=&Fort_Weapon_Projectile_Distant_Mono_Athena_Att
   attenuationOverrides {falloffDistance=2600, attenuationShapeExtents {x=1000}}
  }
 }
 {
  id=Enveloper_4
  children [WavePlayer_12]
  properties {
   pitchCurve {editorCurveData {keys [{time=0}]}}
   volumeCurve {editorCurveData {keys [{interpMode=RCIM_Cubic, tangentMode=RCTM_User, time=0, value=0.1}, {interpMode=RCIM_Cubic, tangentMode=RCTM_User, time=1, value=1}]}}
  }
 }
 {id=Mixer_0, children [SoundClass_2,SoundClass_0], properties {inputVolume [0.7,0.3]}}
 {id=Modulator_0, children [Enveloper_4], properties {pitchMax=1.1}}
 {id=Modulator_2, children [WavePlayer_1], properties {pitchMax=1.3, pitchMin=1.3}}
 {
  id=SoundClass_0
  children [Attenuation_1]
  properties {soundClassOverride=&Fort_WeaponRanged_Fire_Distant}
 }
 {
  id=SoundClass_2
  children [Attenuation_0]
  properties {soundClassOverride=&Fort_WeaponRanged_Fire_3P_PSM}
 }
 {
  id=WavePlayer_1
  properties {bLooping=true, soundWaveAssetPtr=&/Cannon_Fire_CannonBall_InAirLoop_01}
 }
 {id=WavePlayer_12, properties {bLooping=true, soundWaveAssetPtr=/FNGameplayCues/Sounds/Fort_Weapons/Launchers/Quad/quad_rocket_inair_loop_01}}
]
properties {
 bOverrideConcurrency=true
 maxDistance=27500
 volumeMultiplier=0.5
 attenuationOverrides {falloffDistance=5000, reverbWetLevelMax=0.7, reverbWetLevelMin=0.2, attenuationShapeExtents {x=1000}}
 concurrencyOverrides {resolutionRule=PreventNew}
}
