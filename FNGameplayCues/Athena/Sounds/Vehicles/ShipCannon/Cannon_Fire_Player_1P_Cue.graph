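# Sound cue for the first-person ship cannon fire sound heard when the player fires, in the Athena vehicle audio set.
# A Random node picks one of three equally weighted close-range wave variants and feeds a Modulator that boosts volume with slight variation.
# The cue also lowers pitch, raises volume, and uses first-person weapon attenuation, sound class and gunfire concurrency.
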SoundCue /FNGameplayCues/Athena/Sounds/Vehicles/ShipCannon/Cannon_Fire_Player_1P_Cue
root=Modulator_6
nodes [
 {id=Modulator_6, children [Random_3], properties {volumeMax=1.35, volumeMin=1.3}}
 {id=Random_3, children [WavePlayer_9, WavePlayer_10, WavePlayer_11], properties {weights [1,1,1]}}
 {id=WavePlayer_10, properties {soundWaveAssetPtr=/Game/Athena/Sounds/Vehicles/ShipCannon/Cannon_Fire_Player_Close_02}}
 {id=WavePlayer_11, properties {soundWaveAssetPtr=/Game/Athena/Sounds/Vehicles/ShipCannon/Cannon_Fire_Player_Close_01}}
 {id=WavePlayer_9, properties {soundWaveAssetPtr=/Game/Athena/Sounds/Vehicles/ShipCannon/Cannon_Fire_Player_Close_03}}
]
properties {
 attenuationSettings=&Fort_Weapon_Projectile_FirstPerson_Athena_Att
 maxDistance=2097152
 pitchMultiplier=0.9
 soundClassObject=&Fort_WeaponRanged_Fire_1P_PSM
 volumeMultiplier=1.2
 attenuationOverrides {falloffDistance=5000, reverbWetLevelMax=0.7, reverbWetLevelMin=0.2, attenuationShapeExtents {x=1000}}
 concurrencyOverrides {resolutionRule=PreventNew}
 concurrencySet [&Weapon_Gunfire_FirstPerson_Conc]
}
