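# Sound cue for the ship cannon's player landing, in the explosives sound class.
# It mixes two randomly picked layers, a player landing sound and a close cannonball explosion, each behind its own short random delay.
# The mix passes through a modulator for pitch variation, then a medium projectile attenuation with overridden falloff, distance low-pass and no reverb send.
# A concurrency override prevents new instances once the limit is reached.
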SoundCue /FNGameplayCues/Athena/Sounds/Vehicles/ShipCannon/Cannon_Land_Player_Cue
root=Attenuation_1
nodes [
 {
  id=Attenuation_1
  children [Modulator_0]
  properties {
   attenuationSettings=&Fort_Weapon_Projectile_Medium_Mono_Athena_Att
   bOverrideAttenuation=true
   attenuationOverrides {
    bAttenuateWithLPF=true
    bEnableReverbSend=false
    falloffDistance=16000
    lPFFrequencyAtMax=2000
    lPFRadiusMax=12000
    stereoSpread=800
    attenuationShapeExtents {x=2000}
   }
  }
 }
 {id=Delay_0, children [Random_2], properties {delayMax=0.07}}
 {id=Delay_1, children [Random_0], properties {delayMax=0.07}}
 {id=Mixer_0, children [Delay_0,Delay_1], properties {inputVolume [1,1]}}
 {id=Modulator_0, children [Mixer_0], properties {pitchMax=1.1}}
 {id=Random_0, children [WavePlayer_1,WavePlayer_3,WavePlayer_4], properties {weights [1,1,1]}}
 {id=Random_2, children [WavePlayer_0,WavePlayer_2,WavePlayer_5], properties {weights [1,1,1]}}
 {id=WavePlayer_0, properties {soundWaveAssetPtr=&/Cannon_Land_Player_01}}
 {id=WavePlayer_1, properties {soundWaveAssetPtr=&/Cannon_Explosion_CannonBall_Close_03}}
 {id=WavePlayer_2, properties {soundWaveAssetPtr=&/Cannon_Land_Player_02}}
 {id=WavePlayer_3, properties {soundWaveAssetPtr=&/Cannon_Explosion_CannonBall_Close_01}}
 {id=WavePlayer_4, properties {soundWaveAssetPtr=&/Cannon_Explosion_CannonBall_Close_02}}
 {id=WavePlayer_5, properties {soundWaveAssetPtr=&/Cannon_Land_Player_03}}
]
properties {
 bOverrideConcurrency=true
 maxDistance=18000
 soundClassObject=/Game/Sounds/FortSoundClassesAndModes/Fort_Explosives_Medium_PSM
 volumeMultiplier=0.25
 attenuationOverrides {reverbWetLevelMax=0.7, reverbWetLevelMin=0.2}
 concurrencyOverrides {resolutionRule=PreventNew}
}
